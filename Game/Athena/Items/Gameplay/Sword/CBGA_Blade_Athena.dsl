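; Infinity Blade sword pickup actor that players pull out through interaction, refused for blocked tags, vehicles and remote-controlled pawns.
; BeginPlay sets mode-specific draw distance scaled by view distance and spawns the beam at runtime, because an added component fails with spawners.
; Pull start swaps the beam for pull FX; abort stops those and spawns a new beam.
; Death plays pull-out FX; EndPlay destroys the beam and plays a sound in play mode.
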(event EventBeginPlay
  ;[1 1 1] Have to spawn pickup beam instead of adding as a component because adding as a component isn't working when spawning this via the spawner for some reason
  (CallFunction|SetDrawDistance)
  (Networking|CanExecuteCosmeticEvents
    (:True
      (bind _returnvalue (Niagara|SpawnSystemAttached (Variables|FX|GetNS_PickupBeam) (Variables|Default|GetCapsule) "None" (Transformation|GetActorLocation) (Transformation|GetActorRotation) "KeepWorldPosition" true))
      (Variables|Default|SetNS_BeamFX _returnvalue))))

(event EventEndPlay
  (params (EndPlayReason :type EEndPlayReason%20Enum :default "Destroyed"))
  ;[1 1 1] Clean up beam FX
  (Components|DestroyComponent (Variables|Default|GetBeamFX))
  (Utilities|FlowControl|Sequence
    (:then_0
      (Development|PrintString "INfinity Blade - END PICKUP")
      (if (Class|CreativeCommonDeviceFunctionLibrary|IsinPlayMode self)
        (Audio|PlaySoundatLocation "/Game/Sounds/Fort_GamePlay_Sounds/BRSword/AB_Pull_End_Success_Cue.AB_Pull_End_Success_Cue" (Transformation|GetActorLocation))))))

(event EventOnPullStart
  (bind _beamfx (Variables|Default|GetBeamFX))
  (Development|PrintString "Pullstart!!!")
  (Components|Activation|Activate (Variables|Default|GetNS_IBlade_Pull_SwordOnly) true)
  (Utilities|IsValid _beamfx
    (:Is%20Valid
      (Components|Activation|Deactivate _beamfx))))

(event EventOnPullAborted
  (if (Components|Activation|IsActive (Variables|Default|GetNS_IBlade_Pull_SwordOnly))
    (Components|Activation|Deactivate (Variables|Default|GetNS_IBlade_Pull_SwordOnly)))
  (Utilities|IsValid (Variables|Default|GetBeamFX)
    (:Is%20Valid
      (Components|Activation|Deactivate (Variables|Default|GetNS_BeamFX)))
    (:Is%20Not%20Valid
      _))
  (bind _returnvalue (Niagara|SpawnSystemAttached (Variables|FX|GetNS_PickupBeam) (Variables|Default|GetCapsule) "None" (Transformation|GetActorLocation) (Transformation|GetActorRotation) "KeepWorldPosition" true))
  (Variables|Default|SetNS_BeamFX _returnvalue))

(event EventOnDeathPlayEffects
  (params (Damage :type Float%20%28single-precision%29) (DamageTags :type Gameplay%20Tag%20Container%20Structure%20%28by%20ref%29) (Momentum :type Vector :default "0, 0, 0") (HitInfo :type Hit%20Result%20Structure%20%28by%20ref%29) (InstigatedBy :type Fort%20Pawn%20Object%20Reference) (DamageCauser :type Actor%20Object%20Reference) (EffectContext :type Gameplay%20Effect%20Context%20Handle%20Structure))
  (Effects|Components|ParticleSystem|SpawnFXAtLocation (Variables|FX|GetFXSystem_PullOutFX) (Transformation|GetActorLocation)))

(fn BlueprintCanInteract
  (params (InteractingPawn :type Fort%20Pawn%20Object%20Reference) (InteractionBeingAttempted :type EInteractionBeingAttempted%20Enum :default "FirstInteraction") (InteractionType :type TInteractionType%20Enum :default "IT_NoInteraction"))
  (bind (_asfort_remote_controlled_pawn_athena _bsuccess) (Utilities|Casting|CastToFortRemoteControlledPawnAthena (LocalVariables|GetPlayerPawn)))
  (bind _asfort_player_pawn (Utilities|Casting|CastToFortPlayerPawn InteractingPawn)
    (:then
      (LocalVariables|SetPlayerPawn _asfort_player_pawn)
      (return (Math|Boolean|ANDBoolean (not (GameplayTags|HasAnyMatchingGameplayTags (LocalVariables|GetPlayerPawn) (Variables|Default|GetBlockInteractionTags))) (not _bsuccess) (not (RemoteControlPawn|HasControlledRCPawn (LocalVariables|GetPlayerPawn))) (not (Vehicle|IsinVehicle (LocalVariables|GetPlayerPawn))))))
    (:CastFailed
      (return false))))

(fn BlueprintGetInteractionString
  (params (InteractingPawn :type Fort%20Pawn%20Object%20Reference) (InteractionBeingAttempted :type EInteractionBeingAttempted%20Enum :default "FirstInteraction"))
  (return (Variables|GetBuildingActorInteractionText)))

(fn SetDrawDistance
  (bind (_asfort_game_state_athena _bsuccess) (Utilities|Casting|CastToFortGameStateAthena (Game|GetGameState)))
  (for _array_element (Actor|GetComponentsByClass "/Script/Engine.PrimitiveComponent")
    (Utilities|IsValid _array_element
      (:Is%20Valid
        (LOD|SetMaxDrawDistance _array_element (* (select _bsuccess (Variables|MaxDrawDistance|GetMaxDrawDistanceAthena) (Variables|MaxDrawDistance|GetMaxDrawDistanceStw)) (Utilities|Reciprocal (Fortnite|GetNumericCVar "r.ViewDistanceScale"))))))))

(fn ConstructionScript
  (|CallParentFunction))
